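# M_UI_GalileoRealm_Dot is a translucent UI material for a dot element in the Galileo Lobster item's Realm UI.
# It pairs a constant white color with a procedural circle from the shared MF_UI_Circle function.
# An OutlineThickness scalar parameter feeds the circle function, so material instances can adjust the outline width.
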Material /Galileo/Items/Lobster/Assets/UI/Realm/M_UI_GalileoRealm_Dot
graph {
 properties {blendMode=BLEND_Translucent, materialDomain=MD_UI, pixelDepthOffsetMode=PDOM_Legacy, usedShadingModels=Default Lit}
 nodes [
  Constant3Vector() {constant=<rgba 1 1 1 0>}
  MaterialFunctionCall(OutlineThickness=ScalarParameter() {defaultValue=0.05, parameterName=OutlineThickness}) {materialFunction=&MF_UI_Circle}
 ]
}
