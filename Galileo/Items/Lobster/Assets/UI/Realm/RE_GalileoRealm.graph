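# Weapon reticle extension widget for the Galileo Lobster Realm weapon, showing target range and homing lock.
# On Construct it binds to the equipped weapon's UpdateReticleState delegate and resets the reticle.
# Homing changes play or reset the Homing animation of the reticle material's LockedOn and AnimateScale parameters, with ShowingHoming preventing replays.
# Range changes set overlay opacity and center-dot visibility; the red/white UpdateReticleState tint function is not called here.
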WidgetBlueprint /Galileo/Items/Lobster/Assets/UI/Realm/RE_GalileoRealm
animations [Homing {
  binding_count=1
  length_seconds=0.1499
  object=&self:Homing
  sequence {
   bindings [Reticle {
     id=#1
     possessed_class=Image
     tracks [{
       type=MovieSceneWidgetMaterialTrack
       sections [{
         type=MovieSceneParameterSection
         end_frame=4
         start_frame=0
         channels [
          LockedOn {type=MovieSceneScriptingFloatChannel, keys [[0,0],[4.3755,1]]}
          AnimateScale {type=MovieSceneScriptingFloatChannel, keys [[0,0],[4.3755,1]]}
         ]
        }]
      }]
    }]
   display_rate {numerator=30}
  }
 }]
graphs [
 EventGraph {nodes [
   n1 = UpdateReticleState_Event!() -> (OutputDelegate, then, bInRange:bool=~, bHasHomingTarget:bool=~) {is=title K2Node_CustomEvent}
   > SetHomingState!(execute, self, bHoming:bool=@n1.bHasHomingTarget) -> (then)
   > SetInRangeState!(execute, self, bInRange:bool=@n1.bInRange) -> (then)
   Event Construct()
   > n2 = Cast To B_Athena_Galileo_Lobster_Realm!(
    execute
    Object=GetEquippedWeapon!(self:Fort Weapon Reticle Extension Widget Base Object Reference) -> (ReturnValue:Fort Weapon Object Reference)
   ) -> (then, CastFailed, AsB Athena Galileo Lobster Realm:B Athena Galileo Lobster Realm Object Reference) {is=title K2Node_DynamicCast}
   > Assign Update Reticle State!(execute:Exec, self:B Athena Galileo Lobster Realm Object Reference=@n2, Delegate:"Delegate (by ref)"=@n1) -> (then:Exec) {is=title K2Node_AssignDelegate}
   > ResetReticle()
   Event PreConstruct()
   Event Tick()
  ]}
 ResetReticle {nodes [
   ResetReticle() {is=title K2Node_FunctionEntry}
   n1 = Get Homing() -> (Homing:Widget Animation Object Reference)
   > PlayAnimation(InAnimation=@n1, StartAtTime=GetEndTime(self=@n1), PlayMode=Reverse)
   > Set ShowingHoming(ShowingHoming:bool=false) -> (Output_Get:bool)
  ]}
 SetHomingState {nodes [
   n1 = SetHomingState!() -> (then, bHoming:bool=~) {is=title K2Node_FunctionEntry}
   > n2 = Branch(Condition=@n1)
   > Branch(Condition=NOT Boolean(A=Get ShowingHoming() -> (ShowingHoming:bool)))
   > PlayAnimation(InAnimation=Get Homing())
   > Set ShowingHoming(ShowingHoming=true)
   ResetReticle(execute=@n2.else) {is=title K2Node_CallFunction}
  ]}
 SetInRangeState {nodes [
   n1 = SetInRangeState!() -> (then, bInRange:bool=~) {is=title K2Node_FunctionEntry}
   > SetRenderOpacity(self=Get ReticleOverlay() -> (ReticleOverlay:Overlay Object Reference), InOpacity=SelectFloat(A=0.7, B=0.3, bPickA=@n1))
   > SetVisibility(self=Get Center() -> (Center:Image Object Reference), InVisibility=Select(Option 0=Collapsed, Option 1=HitTestInvisible, Index=@n1))
  ]}
 UpdateReticleState {nodes [
   n1 = UpdateReticleState!() -> (then, bLocked:bool=~) {is=title K2Node_FunctionEntry}
   > SetColorAndOpacity(
    self:Image Object Reference=Get Reticle() -> (Reticle:Image Object Reference)
    InColorAndOpacity:Linear Color Structure=SelectColor(A=(R=1,G=0,B=0,A=1), B=(R=1,G=1,B=1,A=1), bPickA=@n1)
   )
  ]}
]
variables [
 Cached Has Homing Target {type=boolean}
 ShowingHoming {type=boolean}
]
widgets {
 info {
  inheritedWidgetCount=1
  parentClass=/Script/FortniteUI.FortWeaponReticleExtensionWidgetBase
  rootWidgetClass=&Overlay
  widgetCount=4
 }
 widgets [
  ReticleOverlay {widgetClassPath=&Overlay, bIsVariable=true, widget=ReticleOverlay}
  Reticle {
   widgetClassPath=&Image
   bIsVariable=true
   parent=ReticleOverlay
   widget=Reticle
   properties {
    visibility=HitTestInvisible
    brush {
     bIsDynamicallyLoaded=false
     drawAs=Image
     imageType=NoImage
     mirroring=NoMirror
     resourceName=None
     resourceObject=&/MI_UI_GalileoRealm_Reticle
     tiling=NoTile
     imageSize=<64 64>
     margin=<ltrb 0 0 0 0>
     outlineSettings {
      bUseBrushTransparency=false
      roundingType=HalfHeightRadius
      width=0
      color {colorUseRule=UseColor_Specified, specifiedColor=<rgba 0 0 0 0>}
      cornerRadii=<0 0 0 0>
     }
     tintColor {colorUseRule=UseColor_Specified, specifiedColor=<rgba 1 1 1 1>}
     uVRegion {bIsValid=false, max=<0 0>, min=<0 0>}
    }
   }
   slot {class=OverlaySlot}
  }
  Center {
   widgetClassPath=&Image
   bIsVariable=true
   parent=ReticleOverlay
   widget=Center
   properties {
    visibility=HitTestInvisible
    brush {
     bIsDynamicallyLoaded=false
     drawAs=Image
     imageType=NoImage
     mirroring=NoMirror
     resourceName=None
     resourceObject=&/MI_UI_GalileoRealm_Dot_Inst
     tiling=NoTile
     imageSize=<15 15>
     margin=<ltrb 0 0 0 0>
     outlineSettings {
      bUseBrushTransparency=false
      roundingType=HalfHeightRadius
      width=0
      color {colorUseRule=UseColor_Specified, specifiedColor=<rgba 0 0 0 0>}
      cornerRadii=<0 0 0 0>
     }
     tintColor {colorUseRule=UseColor_Specified, specifiedColor=<rgba 1 1 1 1>}
     uVRegion {bIsValid=false, max=<0 0>, min=<0 0>}
    }
   }
   slot {class=OverlaySlot, properties {horizontalAlignment=HAlign_Center, verticalAlignment=VAlign_Center}}
  }
  TutorialHighlight {widgetClassPath=&TutorialHighlight}
 ]
}
